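# Unlit additive Niagara material for the shape effect in the main player's firecracker sparks emote.
# Noise-distorted UVs drive panning noise, smoke and caustic textures that blend base, high and low colors, tinted by particle color.
# UV sphere masks shape opacity, with dynamic parameters controlling time offset, desaturation and reveal.
# Shading path switches give mobile cheaper substitutes for distortion, HDR remap, depth/camera fades and animated actor-relative vector-noise curl.
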Material /Game/Animation/Game/MainPlayer/Emotes/FirecrackerSparks/FX/M_FirecrackerEmoteShape
graph {
 properties {
  bUsedWithNiagaraMeshParticles=true
  bUsedWithNiagaraRibbons=true
  bUsedWithNiagaraSprites=true
  blendMode=BLEND_Additive
  pixelDepthOffsetMode=PDOM_Legacy
  shadingModel=MSM_Unlit
  usedShadingModels=Unlit
 }
 nodes [
  Constant() {r=1}
  n1 = Constant2Vector() {r=0.1}
  n2 = AppendVector(A=ScalarParameter() {defaultValue=0.1, parameterName=PanSpeed}, B=Constant())
  n3 = TextureCoordinate() {coordinateIndex=3, uTiling=0.5, vTiling=0.5}
  n4 = ComponentMask(None=TextureSampleParameter2D(UVs=Panner(Coordinate=@n3) {speedX=-0.01}) {parameterName=NoiseTexture, texture=&T_NoisePack_01}.RGB) {g=true, r=true}
  n5 = ScalarParameter() {defaultValue=2, parameterName=UVdistortAmount}
  n6 = Add(
   A=ShadingPathSwitch(Default=Divide(A=@n4, B=@n5) {constB=8}, Mobile=Constant()) {desc=Migrated from FeatureLevelSwitch}
   B=Multiply(A=TextureCoordinate() {coordinateIndex=1, uTiling=5, vTiling=0.7}, B=ScalarParameter() {defaultValue=0.3, parameterName=NoiseTile})
  )
  n7 = Add(A=Time(), B=DynamicParameter() {defaultValue=<rgba 0 1 0.4 0>, paramNames [TimeOffset,Curl_Scale,Desa,AnimReveal]}.TimeOffset)
  n8 = TextureSample(UVs=Panner(Coordinate=@n6, Time=@n7, Speed=Add(A=@n1, B=@n2) {constB=0.2}) {speedX=0.01}) {texture=&T_NoisePack_01}
  n9 = Desaturation(None=@n8.B, Fraction=DynamicParameter() {defaultValue=<rgba 0 1 0.4 0>, paramNames [TimeOffset,Curl_Scale,Desa,AnimReveal]}.Desa) {luminanceFactors=<rgba 0.0143 0.0948 0.1005 0>}
  n10 = Multiply(A=VectorParameter() {parameterName=HighColor, defaultValue=<rgba 1 0.8824 0.757 1>}.RGB, B=@n9)
  n11 = Multiply(A=VectorParameter() {parameterName=LowColor, defaultValue=<rgba 2 0.4606 0.1608 1>}.RGB, B=@n9)
  n12 = Desaturation(None=@n8.G) {luminanceFactors=<rgba 0.0143 0.0948 0 0>}
  n13 = MaterialFunctionCall(Float2=@n6) {materialFunction=&BreakOutFloat2Components}
  n14 = Panner(Coordinate=AppendVector(A=@n13.R, B=Add(A=@n13.G) {constB=0.5}), Time=@n7, Speed=@n2)
  n15 = Constant2Vector() {r=0.03}
  n16 = TextureSample(UVs=Panner(Coordinate=@n6, Time=@n7, Speed=Add(A=@n15, B=@n2) {constB=0.3}) {speedX=0.1}) {texture=&T_ShieldCaustics}
  n17 = LinearInterpolate(A=ScalarParameter() {parameterName=Sorften}, Alpha=@n16.R)
  n18 = TextureCoordinate()
  n19 = SphereMask(
   A=Power(Base=ComponentMask(None=@n18) {r=true}, Exp=ScalarParameter() {defaultValue=0.5, parameterName=OpatcityMaskUbias}) {constExponent=0.5}
   B=Constant() {r=0.5}
  ) {attenuationRadius=0.5, hardnessPercent=0}
  n20 = Multiply(A=@n16.R, B=Multiply(A=@n19, B=@n19))
  n21 = Add(
   A=LinearInterpolate(Alpha=TextureSample(UVs=@n14) {texture=&T_Inky_Smoke_Tile}.R) {constB=3}
   B=LinearInterpolate(Alpha=Add(A=@n17, B=@n20)) {constA=0.2, constB=3}
  )
  n22 = ParticleColor()
  n23 = Multiply(
   A=LinearInterpolate(
    A=VectorParameter() {parameterName=Base Color, defaultValue=<rgba 0.0902 0.0275 0.0118 1>}.RGB
    B=LinearInterpolate(A=@n10, B=@n11, Alpha=@n12)
    Alpha=Clamp(None=@n21)
   )
   B=@n22.RGB
  )
  n24 = Clamp(None=@n23)
  n25 = ScalarParameter() {defaultValue=1, parameterName=Mobile HDR Intensity}
  MaterialFunctionCall(Emissive=ShadingPathSwitch(
    Default=MaterialFunctionCall(LDR Input=@n24, HDR Tint=@n23) {materialFunction=&DeriveHDRfromLDR}.RemappedRange
    Mobile=Multiply(A=@n25, B=@n23)
   ) {desc=Migrated from FeatureLevelSwitch}) {materialFunction=&MF_FX_EmissiveAdjustments}
  n26 = Multiply(
   A=SphereMask(A=TextureCoordinate() {vTiling=0}, B=Constant2Vector() {r=1.02}) {attenuationRadius=0.95, hardnessPercent=55}
   B=SphereMask(A=TextureCoordinate() {vTiling=0}, B=Constant2Vector()) {attenuationRadius=0.95, hardnessPercent=46.4286}
  )
  Multiply(A=@n26) {constB=2}
  Power(Base=@n26)
  Reroute()
  n27 = Multiply(A=@n22.A, B=@n21)
  n28 = DynamicParameter() {defaultValue=<rgba 0 1 0.4 0>, paramNames [TimeOffset,Curl_Scale,Desa,AnimReveal]}
  n29 = SphereMask(
   A=TextureCoordinate() {vTiling=0}
   B=MaterialFunctionCall(X=Add(B=@n28.AnimReveal) {constA=-1}, Y=Constant()) {materialFunction=&MakeFloat2}.Result
  ) {attenuationRadius=1, hardnessPercent=75.9524}
  n30 = Clamp(None=Multiply(A=Multiply(A=@n27, B=@n29), B=@n26))
  n31 = ScalarParameter() {parameterName=CameraFadeDistance}
  ShadingPathSwitch(
   Default=Multiply(
    A=DepthFade(Opacity=@n30)
    B=MaterialFunctionCall(Distance=@n31) {materialFunction=&Near_Camera_Fade}.Result
   )
   Mobile=@n30
  ) {desc=Migrated from FeatureLevelSwitch}
  n32 = WorldPosition()
  n33 = ActorPositionWS()
  n34 = ScalarParameter() {defaultValue=10, parameterName=Curl_Speed}
  n35 = Time()
  n36 = Divide(A=Subtract(A=Subtract(A=@n32.XYZ, B=@n33), B=Multiply(A=@n34, B=@n35)), B=ScalarParameter() {defaultValue=25, parameterName=Curl_Scale})
  ShadingPathSwitch(
   Default=Multiply(A=ScalarParameter() {defaultValue=0.55, parameterName=Curl_Intensity}, B=VectorNoise(World Position=@n36) {noiseFunction=VNF_VectorALU})
   Mobile=Constant() {r=1}
  ) {desc=Migrated from FeatureLevelSwitch}
 ]
}
